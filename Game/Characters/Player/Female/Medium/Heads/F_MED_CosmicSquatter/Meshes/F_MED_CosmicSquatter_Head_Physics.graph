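# Physics asset for the CosmicSquatter female medium player head, giving secondary motion to its hair and fins.
# Capsules on two dyn_hair_04 bones form a chain hanging off a small head capsule, and spheres on the left and right fin bones attach directly to the head.
# Angular limits keep the fins fairly stiff and let the hair chain swing more widely, with little twist on either.
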PhysicsAsset /Game/Characters/Player/Female/Medium/Heads/F_MED_CosmicSquatter/Meshes/F_MED_CosmicSquatter_Head_Physics
bodies {
 dyn_hair_04 [{shapeType=Capsule, length=4.3219, radius=2.7728, shapeName=dyn_hair_04_capsule, center=<3 0 0>, rotation=<rot 90 0 0>}]
 dyn_hair_04_a [{shapeType=Capsule, length=2.5748, radius=1.8688, shapeName=dyn_hair_04_a_capsule, center=<3 0 0>, rotation=<rot 90 0 0>}]
 fin_l [{shapeType=Sphere, radius=4.5, shapeName=fin_l_sphere, center=<0 -3.5 0>}]
 fin_r [{shapeType=Sphere, radius=4.5, shapeName=fin_r_sphere, center=<-3 0 0>}]
 head [{shapeType=Capsule, length=0.505, radius=0.505}]
}
constraints [
 {bone1Name=fin_l, bone2Name=head, swing1LimitDegrees=12, swing2LimitDegrees=12, twistLimitDegrees=3}
 {bone1Name=fin_r, bone2Name=head, swing1LimitDegrees=12, swing2LimitDegrees=12, twistLimitDegrees=3}
 {bone1Name=dyn_hair_04_a, bone2Name=dyn_hair_04, swing1LimitDegrees=35, swing2LimitDegrees=35, twistLimitDegrees=5}
 {bone1Name=dyn_hair_04, bone2Name=head, swing1LimitDegrees=35, swing2LimitDegrees=35, twistLimitDegrees=5}
]
