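# Post-process AnimBlueprint for the female medium CosmicSquatter head, adding secondary physics and procedural rig motion on the shared Fortnite base head skeleton.
# The input pose runs through a RigidBody node simulating the head physics asset relative to spine_05 with overridden gravity, then the head's Control Rig, to the output.
# Two left-eyelid Constraint nodes targeting fin_l and head are disconnected from the output, and the event graph holds only unconnected stub nodes.
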AnimBlueprint /Game/Characters/Player/Female/Medium/Heads/F_MED_CosmicSquatter/Meshes/F_Med_CosmicSquatter_Head_Post_Process
target_skeleton=/Game/Characters/Player/Common/Fortnite_Base_Head/Fortnite_Base_Head_Export_Skeleton
graphs [
 AnimGraph {nodes [
   n1 = "Constraint - L_eye_lid_upper_mid (PP)"!(ComponentPose, Alpha=1, ConstraintWeights_0=1, ConstraintWeights_1:float=0) -> (Pose) {
    alphaBoolBlend {blendOption=Linear}
    boneToModify {boneName=L_eye_lid_upper_mid}
    constraintSetup [{targetBone {boneName=fin_l}}, {targetBone {boneName=head}}]
    constraintWeights [1,1]
   }
   "Constraint - L_eye_lid_lower_mid (PP)"!(ComponentPose=@n1, Alpha=1, ConstraintWeights_0=1, ConstraintWeights_1:float=0) -> (Pose) {
    alphaBoolBlend {blendOption=Linear}
    boneToModify {boneName=L_eye_lid_lower_mid}
    constraintSetup [{targetBone {boneName=fin_l}}, {targetBone {boneName=head}}]
    constraintWeights [1,1]
   }
   n2 = RigidBody(ComponentPose=Local To Component(LocalPose=Input Pose())) {
    bClampLinearTranslationLimitToRefPose=true
    bOverrideWorldGravity=true
    overridePhysicsAsset=&/F_MED_CosmicSquatter_Head_Physics
    simulationSpace=BaseBoneSpace
    alphaBoolBlend {blendOption=Linear}
    baseBoneRef {boneName=spine_05}
    componentLinearAccScale=<0.8 0.8 0.8>
    componentLinearVelScale=<0.8 0.8 0.8>
    overrideWorldGravity {z=-980}
   }
   n3 = Control Rig(Source=Component To Local(ComponentPose=@n2)) {
    alphaBoolBlend {blendOption=Linear}
    controlRigAssetReference {blueprintRigClass=&/F_MED_CosmicSquatter_Head_CtrlRig.F_MED_CosmicSquatter_Head_CtrlRig_C}
   }
   Output Pose(Result=@n3) {layerGroup=None}
  ]}
 EventGraph {nodes [
   TryGetPawnOwner()
   Event BlueprintUpdateAnimation()
  ]}
]
variables []
